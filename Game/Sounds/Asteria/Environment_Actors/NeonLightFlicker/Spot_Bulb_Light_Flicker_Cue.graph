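# Sound cue for a flickering spot bulb light, playing as a positional environmental ambience.
# A mixer layers two modulated loops, each a short fixed delay before a random pick: single flickers, and triplet flickers whose loop starts seconds later.
# The first two quality levels share this mixer, and the other two are empty.
# Overrides add occlusion, listener-focus attenuation and a voice limit, with a send to the exterior ambient bus.
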SoundCue /Game/Sounds/Asteria/Environment_Actors/NeonLightFlicker/Spot_Bulb_Light_Flicker_Cue
root=QualityLevel_0
nodes [
 {id=Delay_0, children [Random_1], properties {delayMax=0.33, delayMin=0.33}}
 {id=Delay_1, children [Random_0], properties {delayMax=0.33, delayMin=0.33}}
 {id=Delay_3, children [Modulator_0], properties {delayMax=3, delayMin=3}}
 {id=Looping_0, children [Delay_0]}
 {id=Looping_1, children [Delay_1]}
 {id=Mixer_0, children [Modulator_1,Delay_3], properties {inputVolume [1,1]}}
 {id=Modulator_0, children [Looping_0], properties {pitchMax=1}}
 {id=Modulator_1, children [Looping_1], properties {pitchMax=1}}
 {id=QualityLevel_0, children [Mixer_0,Mixer_0,null,null]}
 {id=Random_0, children [WavePlayer_1,WavePlayer_3,WavePlayer_4,WavePlayer_5], properties {weights [1,1,1,1]}}
 {id=Random_1, children [WavePlayer_6,WavePlayer_7,WavePlayer_8,WavePlayer_9], properties {weights [1,1,1,1]}}
 {id=WavePlayer_1, properties {soundWaveAssetPtr=&/Spot_Bulb_Light_Flicker_Single_01}}
 {id=WavePlayer_3, properties {soundWaveAssetPtr=&/Spot_Bulb_Light_Flicker_Single_03}}
 {id=WavePlayer_4, properties {soundWaveAssetPtr=&/Spot_Bulb_Light_Flicker_Single_04}}
 {id=WavePlayer_5, properties {soundWaveAssetPtr=&/Spot_Bulb_Light_Flicker_Single_05}}
 {id=WavePlayer_6, properties {soundWaveAssetPtr=&/Spot_Bulb_Light_Flicker_Tripplet_01}}
 {id=WavePlayer_7, properties {soundWaveAssetPtr=&/Spot_Bulb_Light_Flicker_Tripplet_02}}
 {id=WavePlayer_8, properties {soundWaveAssetPtr=&/Spot_Bulb_Light_Flicker_Tripplet_03}}
 {id=WavePlayer_9, properties {soundWaveAssetPtr=&/Spot_Bulb_Light_Flicker_Tripplet_04}}
]
properties {
 bEnableSubmixSends=false
 bOverrideAttenuation=true
 bOverrideConcurrency=true
 maxDistance=1350
 soundClassObject=&Fort_Ambient_OneShots
 volumeMultiplier=0.351
 attenuationOverrides {
  bEnableListenerFocus=true
  bEnableOcclusion=true
  distanceAlgorithm=NaturalSound
  falloffDistance=1250
  nonFocusVolumeAttenuation=0.7
  occlusionLowPassFilterFrequency=2000
  occlusionTraceChannel=ECC_GameTraceChannel11
  occlusionVolumeAttenuation=0.5
  stereoSpread=100
  attenuationShapeExtents {x=100}
  submixSendSettings [{
    bEnableHPFCutoff=false
    bEnableLPFCutoff=false
    disableManualSendClamp=false
    hPFCutoff=20
    lPFCutoff=20000
    maxSendDistance=6000
    maxSendLevel=1
    minSendDistance=400
    minSendLevel=0
    sendLevel=0.2
    sendLevelControlMethod=Linear
    soundSubmix=null
   }]
 }
 busSends [{
   audioBus=null
   maxSendDistance=1500
   maxSendLevel=0.6
   minSendLevel=0.3
   sendLevel=1
   soundSourceBus=&AmbientAudio_Exterior_SourceBus
  }]
 concurrencyOverrides {maxCount=6, resolutionRule=StopLowestPriority, voiceStealReleaseTime=2}
}
